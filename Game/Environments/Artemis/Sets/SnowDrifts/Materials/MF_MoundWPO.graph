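# Material function for snow drift mounds that pushes vertices along their normals, apparently for a breathing effect, added to existing world position offset.
# A sphere mask around an offset object pivot limits displacement, whose amount blends from close to far values by camera distance.
# The BreathingControl scalar scales the result, while the time-based snore sine is left unconnected.
# The UseMoundWPO static switch otherwise passes attributes through unchanged.
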MaterialFunction /Game/Environments/Artemis/Sets/SnowDrifts/Materials/MF_MoundWPO
graph {nodes [
  n1 = Time()
  n2 = Divide(A=ScalarParameter() {defaultValue=1, group=BC_WPO, parameterName=SnoreSpeed}, B=ScalarParameter() {defaultValue=10, parameterName=Duration})
  ConstantBiasScale(None=Sine(None=Multiply(A=@n1, B=@n2)))
  n3 = GetMaterialAttributes(None=FunctionInput() {inputName=Material, inputType=FunctionInput_MaterialAttributes}) {attributeGetTypes [#1]}
  n4 = Divide(A=Distance(A=ObjectPositionWS(), B=CameraPositionWS()), B=ScalarParameter() {defaultValue=2048, parameterName=ClampingDistance})
  n5 = MaterialFunctionCall() {materialFunction=&ObjectPivotPoint}
  n6 = VectorParameter() {group=BC_WPO, parameterName=SphereMask_Offset, defaultValue=<rgba 144 96 512 1>}
  n7 = LinearInterpolate(
   B=LinearInterpolate(
    A=VectorParameter() {group=BC_WPO, parameterName=WPO_Offset_Amt_CLose, defaultValue=<rgba 12 12 12 1>}.RGB
    B=VectorParameter() {group=BC_WPO, parameterName=WPO_Offset_Amt, defaultValue=<rgba 22 22 22 1>}.RGB
    Alpha=Clamp(None=@n4)
   )
   Alpha=SphereMask(
    A=WorldPosition().XYZ
    B=Add(A=@n5.Object Pivot Location, B=@n6.RGB)
    Radius=ScalarParameter() {defaultValue=1024, group=BC_WPO, parameterName=MaskRadius}
    Hardness=Constant()
   )
  )
  n8 = ScalarParameter() {group=BC_WPO, parameterName=BreathingControl}
  n9 = Add(A=@n3.WorldPositionOffset, B=Multiply(A=VertexNormalWS(), B=Multiply(A=@n7, B=@n8)))
  FunctionOutput(None=StaticSwitchParameter(
    True=SetMaterialAttributes!(MaterialAttributes=@n3.MaterialAttributes, WorldPositionOffset=@n9) {attributeSetTypes [#1]}
    False=@n3.MaterialAttributes
   ) {parameterName=UseMoundWPO})
 ]}
